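# Niagara system for the fuse effect on TheBomb face accessory, a cosmetic hat.
# It has two emitters, SparksonWick and Light, whose renderers are not listed here.
# It exposes a User.bisHeadHidden bool, presumably for reacting to a hidden head.
# System scripts use SystemState and a spawn-time Set Variables module, alongside bool-to-float and float-to-vector dynamic inputs.
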NiagaraSystem /BRCosmetics/Accessories/Hats/M_MED_TheBomb_FaceAcc/FX/NS_TheBomb_Fuse
dependencies {
 usedDataInterfaces []
 usedDynamicInputs [&VectorFromFloat, &MakeCustomFloatFromBool]
 usedModules [
  &SystemState
  &self:SystemSpawnScript.SystemScriptSource.SystemScriptGraph.NiagaraNodeAssignment_0.SetVariables_#1
 ]
}
emitters {Light {rendererClasses [], renderers []}, SparksonWick {rendererClasses [], renderers []}}
summary {
 systemName=NS_TheBomb_Fuse
 emitters [{emitterName=SparksonWick, rendererClasses []}, {emitterName=Light, rendererClasses []}]
 userVariables [User.bisHeadHidden {type=NiagaraBool}]
}
