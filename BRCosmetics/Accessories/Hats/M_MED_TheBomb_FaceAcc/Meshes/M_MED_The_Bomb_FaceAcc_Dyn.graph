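# Physics asset for The Bomb face accessory cosmetic, providing bodies and joints for its dynamic fuse.
# Small capsule bodies cover spine_05, neck_02, head, head_alt and two fuse bones.
# The fuse chains from head_alt with swing limits and much tighter twist limits.
# Head joins spine_05 through neck_02, while head_alt joins spine_05 directly with free swing and twist.
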PhysicsAsset /BRCosmetics/Accessories/Hats/M_MED_TheBomb_FaceAcc/Meshes/M_MED_The_Bomb_FaceAcc_Dyn
bodies {
 dyn_fuse_01 [{shapeType=Capsule, length=2, radius=1, rotation=<rot 90 0 0>}]
 dyn_fuse_02 [{shapeType=Capsule, length=2, radius=1, center=<0.5 0 0>, rotation=<rot 90 0 0>}]
 head [{shapeType=Capsule, length=1, radius=1}]
 head_alt [{shapeType=Capsule, length=1, radius=1}]
 neck_02 [{shapeType=Capsule, length=1, radius=1}]
 spine_05 [{shapeType=Capsule, length=1, radius=1}]
}
constraints [
 {bone1Name=neck_02, bone2Name=spine_05, swing1LimitDegrees=45, swing2LimitDegrees=45, twistLimitDegrees=45}
 {bone1Name=head, bone2Name=neck_02, swing1LimitDegrees=45, swing2LimitDegrees=45, twistLimitDegrees=45}
 {bone1Name=dyn_fuse_01, bone2Name=head_alt, swing1LimitDegrees=45, swing2LimitDegrees=45, twistLimitDegrees=5}
 {bone1Name=dyn_fuse_02, bone2Name=dyn_fuse_01, swing1LimitDegrees=45, swing2LimitDegrees=45, twistLimitDegrees=5}
 {
  bone1Name=head_alt
  bone2Name=spine_05
  swing1LimitDegrees=45
  swing1Motion=Free
  swing2LimitDegrees=45
  swing2Motion=Free
  twistLimitDegrees=45
  twistMotion=Free
 }
]
